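; Season 29 Battle Pass quest ability that tracks the owning player entering and leaving the Cerberus POI, apparently while "smelly".
; On activation it calls the parent, then starts wait-tag-added and wait-tag-removed tasks side by side on the Cerberus POI tag.
; Each task applies its Entered or Exited gameplay effect to the player and prints a debug string.
; The end-ability event is empty.
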(event Ability|EventActivateAbility
  (bind _incerberuspoi (Variables|Default|GetInCerberusPOI))
  (|CallParentFunction)
  (Ability|Tasks|WaitGameplayTagAdd _incerberuspoi
    (:then
      (Ability|Tasks|WaitGameplayTagRemove _incerberuspoi
        (:Removed
          (Development|PrintString "I'm not near Cerberus" true true "(R=0.911458,G=0.000000,B=0.166987,A=1.000000)")
          (GameplayEffects|ApplyGameplayEffectToSelf :self (Ability|GetFortAbilitySystemComponent (Variables|GAVars|Data|GetOwningPlayerPawn)) :GameplayEffectClass (Variables|Default|GetGE_ExitedCerberusPOI) :Level 0.0))))
    (:Added
      (GameplayEffects|ApplyGameplayEffectToSelf :self (Ability|GetFortAbilitySystemComponent (Variables|GAVars|Data|GetOwningPlayerPawn)) :GameplayEffectClass (Variables|Default|GetGE_EnteredCerberusPOI) :Level 0.0)
      (Development|PrintString "I'm near Cerberus and Smelly" true true "(R=0.082170,G=0.161458,B=0.000000,A=1.000000)"))))

(event Ability|EventOnEndAbility
  (params (bWasCancelled :type Boolean)))
